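# Translucent UI material that draws the background for the ColdDay event screen, styled as a map of islands and sea.
# A wave-textured, gradient-shaded sea and a map gradient are built for blending through offset island texture masks into Final BG.
# Smoke color gradients blend by UseFullSmokeBackground, alongside a Levels stage with range, gamma and hue shift.
# An Is3DSpace static switch selects viewport or UI UVs, which are rescaled to fill or fit.
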Material /Game/Athena/UI/Events/ColdDay/Materials/M_UI_ColdDay_BG
graph {
 properties {
  bUsedWithStaticMesh=false
  blendMode=BLEND_Translucent
  materialDomain=MD_UI
  pixelDepthOffsetMode=PDOM_Legacy
  usedShadingModels=Default Lit
 }
 nodes [
  n1 = ScalarParameter() {defaultValue=1, parameterName=UseFullSmokeBackground, sliderMax=1}
  LinearInterpolate(A=NamedRerouteUsage(), Alpha=@n1)
  MaterialFunctionCall() {materialFunction=&MF_UI_UVScaling}
  NamedRerouteDeclaration(None=StaticSwitchParameter(
    True=ScreenPosition().ViewportUV
    False=MaterialFunctionCall() {materialFunction=&GetUserInterfaceUV}.Normalized UV
   ) {group=UV Space, parameterName=Is3DSpace}) {name=Normal UV, nodeColor=<rgba 0 0 0 1>}
  n2 = MaterialFunctionCall!(ScreenUV=NamedRerouteUsage(), TextureSize, ScalePivotPoint) {materialFunction=/FortUILibrary/MaterialFunctions/Basics/MF_UI_ScreenSpaceUVScaling}
  NamedRerouteDeclaration(None=@n2.Scaled to Fill) {name=Normal UV Scaled to Fill, nodeColor=<rgba 0 0 0 1>}
  n3 = NamedRerouteUsage()
  n4 = MaterialFunctionCall(UVs=@n3, GradientStart=Constant(), GradientEnd=Constant() {r=1}, RotationAngle=Constant() {r=0.25}) {materialFunction=&MF_UI_LinearGradient}
  NamedRerouteDeclaration(None=LinearInterpolate(
    A=Constant3Vector() {constant=<rgba 0.6458 0.5485 0.407 1>}
    B=Constant3Vector() {constant=<rgba 0.8229 0.76 0.6686 1>}
    Alpha=SmoothStep(Value=@n4.Result)
   )) {name=Map gradient, nodeColor=<rgba 0 0.2969 1 1>}
  n5 = Constant3Vector() {constant=<rgba 0.1979 0.1219 0.0072 1>}
  n6 = NamedRerouteUsage()
  n7 = Constant() {r=0.05}
  n8 = LinearInterpolate(
   A=Constant3Vector() {constant=<rgba 0.7682 0.6584 0.4969 1>}
   B=Constant3Vector() {constant=<rgba 0.7135 0.5964 0.4349 1>}
   Alpha=TextureSample(UVs=MaterialFunctionCall(UVs=@n6, Texture Scale=@n7) {materialFunction=&ScaleUVsByCenter}.UVs) {samplerType=SAMPLERTYPE_LinearColor, texture=&/T_UI_ColdDay_Waves}.A
  )
  n9 = SmoothStep(Value=MaterialFunctionCall(UV=NamedRerouteUsage(), GradientStart=Constant() {r=-0.5}, GradientEnd=Constant() {r=1}) {materialFunction=&MF_UI_RadialGradient}.Smooth) {constMin=-0.25}
  n10 = ComponentMask(None=MaterialFunctionCall(UVs=NamedRerouteUsage(), "Rotation Angle (0-1)"=Constant() {r=0.75}) {materialFunction=&CustomRotator}.Rotated Values) {r=true}
  NamedRerouteDeclaration(None=LinearInterpolate(
    A=LinearInterpolate(A=@n5, B=@n8, Alpha=@n9)
    B=Constant3Vector() {constant=<rgba 0.4564 0.3613 0.227 1>}
    Alpha=SmoothStep(Value=@n10) {constMin=0.25}
   )) {name=Sea Background, nodeColor=<rgba 0.5469 1 0 1>}
  n11 = VectorParameter() {group=Fg_Smoke, parameterName=TransparentSmoke_ColorLight, defaultValue=<rgba 0.15 0.0103 0.2159 1>}
  n12 = VectorParameter() {group=Fg_Smoke, parameterName=TransparentSmoke_ColorDark, defaultValue=<rgba 0.0137 0.0015 0.0685 1>}
  n13 = NamedRerouteUsage()
  n14 = Constant()
  n15 = Constant() {r=0.75}
  n16 = Constant() {r=0.25}
  n17 = Constant() {r=-0.15}
  n18 = NamedRerouteUsage()
  n19 = NamedRerouteUsage()
  n20 = LinearInterpolate(
   A=LinearInterpolate(
    A=MaterialFunctionCall(R=@n11.R, G=@n11.R, B=@n11.R) {materialFunction=&AppendMany}.RGBA
    B=MaterialFunctionCall(R=@n12.R, G=@n12.R, B=@n12.R) {materialFunction=&AppendMany}.RGBA
    Alpha=MaterialFunctionCall(UVs=@n13, GradientStart=@n14, GradientEnd=@n15, RotationAngle=@n16, GradientOffset=@n17) {materialFunction=&MF_UI_LinearGradient}.Result
   )
   B=Saturate(None=Add(A=@n18, B=@n19))
   Alpha=@n1
  )
  NamedRerouteDeclaration(None=@n20) {name=ApplyLevels, nodeColor=<rgba 0.0547 0 1 1>}
  n21 = Constant()
  n22 = Subtract(A=NamedRerouteUsage(), B=@n21)
  n23 = Constant()
  n24 = ScalarParameter() {defaultValue=1, group=Levels, parameterName=LevelsMax}
  n25 = Min(A=Divide(A=Max(A=@n22, B=@n23), B=Subtract(A=@n24, B=@n21)), B=Constant() {r=1}) {desc=LevelsControlInputRange}
  n26 = Divide(A=Constant() {r=1}, B=ScalarParameter() {defaultValue=1, group=Levels, parameterName=LevelsGamma})
  NamedRerouteDeclaration(None=MaterialFunctionCall(
    Hue Shift Percentage=ScalarParameter() {defaultValue=-0.005, group=Levels, parameterName=LevelsHueShift, sliderMax=0.05, sliderMin=-0.05}
    Texture=Power(Base=@n25, Exp=@n26) {desc=LevelsControlInput}
   ) {materialFunction=&HueShift}.Result) {name=LevelsApplied, nodeColor=<rgba 0 1 0.2734 1>}
  n27 = Constant() {r=0.08}
  n28 = Constant() {r=-0.08}
  n29 = TextureSample(UVs=Add(A=@n3, B=AppendVector(A=@n27, B=@n28))) {samplerType=SAMPLERTYPE_LinearColor, texture=&/T_UI_Islands_01}
  n30 = Saturate(None=Multiply(A=Subtract(A=@n29.R) {constB=0.04}))
  n31 = NamedRerouteUsage()
  n32 = AppendVector(A=Constant(), B=Constant() {r=0.2})
  n33 = SmoothStep(Value=TextureSample(UVs=Add(A=@n31, B=@n32)) {samplerType=SAMPLERTYPE_LinearColor, texture=&T_UI_Islands_01}.G) {constMin=0.4}
  NamedRerouteDeclaration(None=LinearInterpolate(A=NamedRerouteUsage(), B=NamedRerouteUsage(), Alpha=Max(A=@n30, B=@n33))) {name=Final BG, nodeColor=<rgba 0 0 0 1>}
  NamedRerouteDeclaration(None=@n2.Scaled to Fit) {name=Normal UV Scaled to Fit, nodeColor=<rgba 0 0 0 1>}
  NamedRerouteUsage()
  NamedRerouteUsage()
  n34 = Add(
   A=MaterialFunctionCall() {materialFunction=&MF_UI_ScreenSpaceUVScaling}.Scaled to Fill
   B=AppendVector(A=Constant(), B=Constant() {r=0.2})
  )
  n35 = Constant() {r=-0.2}
  n36 = Constant() {r=1}
  SmoothStep(Value=LinearInterpolate(
    A=Constant3Vector() {constant=<rgba 0.2664 0.1878 0.107 1>}
    B=Constant3Vector() {constant=<rgba 0.6795 0.5583 0.402 1>}
    Alpha=MaterialFunctionCall(UV=@n34, GradientStart=@n35, GradientEnd=@n36) {materialFunction=&MF_UI_RadialGradient}.Smooth
   ))
  SmoothStep(Value=@n29.R) {constMin=0.8}
  StaticSwitchParameter(True=NamedRerouteUsage(), False=@n20) {group=UV Space, parameterName=Is3DSpace}
 ]
}
