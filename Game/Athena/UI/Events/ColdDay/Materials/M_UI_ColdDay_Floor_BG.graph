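# Floor and background material for the ColdDay UI event, feeding a shared floor setup function, with a shop gradient variant blended in by an Is Shop parameter.
# It builds a sea backdrop from radial gradients and a wave texture, then composites two movable, scalable island map layers masked by linear and radial fades.
# Screen-space fit and fill UVs, named reroutes and a gamma and hue-shift levels stage organize the color pipeline.
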Material /Game/Athena/UI/Events/ColdDay/Materials/M_UI_ColdDay_Floor_BG
graph {
 properties {pixelDepthOffsetMode=PDOM_Legacy, usedShadingModels=Default Lit}
 nodes [
  Constant() {r=1}
  n1 = VectorParameter() {group=Shop, parameterName=Store Gradient Color, defaultValue=<rgba 1 0 0.9017 1>}
  n2 = NamedRerouteUsage()
  n3 = SmoothStep(Value=MaterialFunctionCall(
    UVs=NamedRerouteUsage()
    GradientStart=Constant()
    GradientEnd=ScalarParameter() {defaultValue=1, group=Shop, parameterName=Gradient End}
    RotationAngle=Constant() {r=0.31}
   ) {materialFunction=&MF_UI_LinearGradient}.Result) {constMax=1.2}
  MaterialFunctionCall!(
   Color=LinearInterpolate(A=NamedRerouteUsage(), B=LinearInterpolate(A=@n1.RGB, B=@n2, Alpha=@n3), Alpha=ScalarParameter() {group=Shop, parameterName=Is Shop})
   Color Intensity=ScalarParameter() {defaultValue=0.95, group=0 - Global, parameterName=Color Intensity}
   Floor Radius=ScalarParameter() {defaultValue=0.175, group=6 - Floor, parameterName=Floor Radius}
   Floor Density=ScalarParameter() {defaultValue=4, group=6 - Floor, parameterName=Floor Density}
   Floor Metallic=ScalarParameter() {group=6 - Floor, parameterName=Floor Metalic}
   Floor Specular=ScalarParameter() {group=6 - Floor, parameterName=Floor Spec}
   Floor Roughness=ScalarParameter() {defaultValue=1, group=6 - Floor, parameterName=Floor Roughness}
   IsFloor=StaticBoolParameter() {group=6 - Floor, parameterName=IsFloor}
  ) {materialFunction=/Game/UI/Foundation/Environment/Shell/Materials/MaterialFunctions/MF_UI_BP_FloorSetup}
  n4 = ScalarParameter() {defaultValue=-0.1, group=Map Transform, parameterName=Map X Location}
  n5 = Add(A=Constant() {r=0.08}, B=@n4) {constB=-0.1}
  n6 = ScalarParameter() {group=Map Transform, parameterName=Map Y Location}
  n7 = Add(A=@n6) {constB=-0.08}
  n8 = TextureSample(UVs=Add(A=NamedRerouteUsage(), B=AppendVector(A=@n5, B=@n7))) {samplerType=SAMPLERTYPE_LinearColor, texture=&/T_UI_Islands_01}
  n9 = Saturate(None=Multiply(A=Subtract(A=@n8.R) {constB=0.04}))
  n10 = Add(A=Constant(), B=@n4) {constB=-0.1}
  n11 = Add(A=@n6) {constB=0.2}
  n12 = TextureSample(UVs=Add(A=NamedRerouteUsage(), B=AppendVector(A=@n10, B=@n11))) {samplerType=SAMPLERTYPE_LinearColor, texture=&T_UI_Islands_01}
  n13 = Saturate(None=Multiply(A=Subtract(A=@n12.G) {constB=0.04}))
  Min(A=@n9, B=@n13)
  NamedRerouteDeclaration(None=MaterialFunctionCall(UVs=NamedRerouteUsage(), Texture Scale=ScalarParameter() {defaultValue=0.5, group=Map Transform, parameterName=Map Scale}) {materialFunction=&ScaleUVsByCenter}.UVs) {name=Map UVs, nodeColor=<rgba 1 0.3047 0 1>}
  n14 = SmoothStep(
   Max=ScalarParameter() {defaultValue=0.5, group=Map, parameterName=Map Fade Amount}
   Value=MaterialFunctionCall(UVs=NamedRerouteUsage(), GradientStart=Constant() {r=0.25}, GradientEnd=Constant() {r=1}, RotationAngle=Constant() {r=0.25}) {materialFunction=&MF_UI_LinearGradient}.Result
  ) {constMax=0.5}
  n15 = ScalarParameter() {defaultValue=0.5, group=Map, parameterName=Map Fade Opacity}
  n16 = Max(A=@n9, B=@n13)
  n17 = MaterialFunctionCall(UVs=NamedRerouteUsage(), Radius=Constant() {r=0.6}) {materialFunction=&RadialGradientExponential}
  n18 = Min(A=Min(A=Multiply(A=@n14, B=@n15) {constB=0.4}, B=@n16), B=SmoothStep(Value=Min(A=@n17.RadialGradientExponential, B=@n16)))
  n19 = LinearInterpolate(A=NamedRerouteUsage(), B=NamedRerouteUsage(), Alpha=@n18)
  NamedRerouteDeclaration(None=@n19) {name=Final BG, nodeColor=<rgba 0.3828 0 1 1>}
  NamedRerouteDeclaration(None=LinearInterpolate(A=Constant3Vector() {constant=<rgba 0.3438 0.2065 0.097 1>}, B=NamedRerouteUsage(), Alpha=@n18)) {name=Map Fill, nodeColor=<rgba 0 0.2969 1 1>}
  n20 = Constant3Vector() {constant=<rgba 0.2708 0.1862 0.0776 1>}
  n21 = VectorParameter() {group=Waves, parameterName=Wave Light Color, defaultValue=<rgba 0.7682 0.6584 0.4969 1>}
  n22 = VectorParameter() {group=Waves, parameterName=Wave Dark, defaultValue=<rgba 0.6719 0.5615 0.4095 1>}
  n23 = MaterialFunctionCall(UVs=NamedRerouteUsage(), Texture Scale=ScalarParameter() {defaultValue=0.03, group=Waves, parameterName=Wave Scale}) {materialFunction=&ScaleUVsByCenter}
  n24 = LinearInterpolate(
   A=@n21.RGB
   B=@n22.RGB
   Alpha=Max(
    A=TextureSample(UVs=@n23.UVs) {samplerType=SAMPLERTYPE_LinearColor, texture=&/T_UI_ColdDay_Waves}.A
    B=NamedRerouteUsage()
   )
  )
  n25 = NamedRerouteUsage()
  n26 = Constant() {r=-0.25}
  n27 = Constant() {r=1}
  n28 = Divide(A=SmoothStep(Value=MaterialFunctionCall(UV=@n25, GradientStart=@n26, GradientEnd=@n27) {materialFunction=&MF_UI_RadialGradient}.Smooth) {constMin=-0.154})
  n29 = Constant() {r=0.05}
  n30 = Constant() {r=1}
  n31 = Add(
   A=MaterialFunctionCall(UVs=NamedRerouteUsage(), Texture Scale=Constant() {r=1}) {materialFunction=&ScaleUVsByCenter}.UVs
   B=AppendVector(A=Constant(), B=Constant() {r=0.2})
  )
  n32 = SmoothStep(Value=LinearInterpolate(
    A=@n20
    B=Constant3Vector() {constant=<rgba 0.6795 0.5583 0.402 1>}
    Alpha=MaterialFunctionCall(UV=@n31, GradientStart=@n29, GradientEnd=@n30) {materialFunction=&MF_UI_RadialGradient}.Smooth
   ))
  n33 = NamedRerouteUsage()
  n34 = Constant() {r=0.75}
  n35 = SmoothStep(Value=ComponentMask(None=MaterialFunctionCall(UVs=@n33, "Rotation Angle (0-1)"=@n34) {materialFunction=&CustomRotator}.Rotated Values) {r=true}) {constMax=1.2, constMin=0.4}
  NamedRerouteDeclaration(None=LinearInterpolate(A=LinearInterpolate(A=@n20, B=@n24, Alpha=@n28), B=@n32, Alpha=@n35)) {name=Sea Background, nodeColor=<rgba 0.5469 1 0 1>}
  NamedRerouteDeclaration(None=@n16) {name=Map Mask, nodeColor=<rgba 1 0.7266 0 1>}
  n36 = MaterialFunctionCall!(ScreenUV, TextureSize, ScalePivotPoint) {materialFunction=/FortUILibrary/MaterialFunctions/Basics/MF_UI_ScreenSpaceUVScaling}
  NamedRerouteDeclaration(None=@n36.Scaled to Fit) {name=Normal UV Scaled to Fit, nodeColor=<rgba 0 0 0 1>}
  NamedRerouteDeclaration(None=@n36.Scaled to Fill) {name=Normal UV Scaled to Fill, nodeColor=<rgba 0 0 0 1>}
  n37 = Constant()
  n38 = Subtract(A=NamedRerouteUsage(), B=@n37)
  n39 = Constant()
  n40 = ScalarParameter() {defaultValue=1, group=Levels, parameterName=LevelsMax}
  n41 = Min(A=Divide(A=Max(A=@n38, B=@n39), B=Subtract(A=@n40, B=@n37)), B=Constant() {r=1}) {desc=LevelsControlInputRange}
  n42 = Divide(A=Constant() {r=1}, B=ScalarParameter() {defaultValue=1.05, group=Levels, parameterName=LevelsGamma})
  NamedRerouteDeclaration(None=MaterialFunctionCall(
    Hue Shift Percentage=ScalarParameter() {defaultValue=-0.005, group=Levels, parameterName=LevelsHueShift, sliderMax=0.05, sliderMin=-0.05}
    Texture=Power(Base=@n41, Exp=@n42) {desc=LevelsControlInput}
   ) {materialFunction=&HueShift}.Result) {name=LevelsApplied, nodeColor=<rgba 0 1 0.2734 1>}
  NamedRerouteDeclaration(None=@n19) {name=Map Combined BG, nodeColor=<rgba 0 0 0 1>}
  NamedRerouteDeclaration(None=LinearInterpolate(A=LinearInterpolate(A=@n21.RGB, B=@n22.RGB, Alpha=@n28), B=@n32, Alpha=@n35)) {name=Map Gradient, nodeColor=<rgba 0 1 0.625 1>}
  NamedRerouteUsage()
  NamedRerouteUsage()
  NamedRerouteUsage()
  VectorParameter() {group=Map, parameterName=Map Fill Top, defaultValue=<rgba 0.8229 0.76 0.6686 1>}
  VectorParameter() {group=Map, parameterName=Map Fill, defaultValue=<rgba 0.3968 0.2384 0.1119 1>}
 ]
}
